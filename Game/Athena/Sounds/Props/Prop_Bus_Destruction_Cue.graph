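# Sound cue for a bus prop destruction sound, classed under Fort_Impacts.
# A single wave player for Prop_Bus_Destruction feeds a Modulator root, which applies a randomized, strongly boosted volume on each play.
# The cue scales overall volume slightly down and uses the shared Fort_Prop_Spooky attenuation asset for spatial falloff.
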SoundCue /Game/Athena/Sounds/Props/Prop_Bus_Destruction_Cue
root=Modulator_0
nodes [
 {id=Modulator_0, children [WavePlayer_0], properties {volumeMax=8, volumeMin=7}}
 {id=WavePlayer_0, properties {soundWaveAssetPtr=&/Prop_Bus_Destruction}}
]
properties {
 attenuationSettings=/Game/Sounds/Attenuation/Fort_Prop_Spooky
 maxDistance=2000
 soundClassObject=&Fort_Impacts
 volumeMultiplier=0.8
}
