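# Sound cue that plays the second bus horn flavor sound as an ambient prop effect in Athena.
# A single wave player feeds a modulator left at default settings, which randomizes pitch and volume slightly on each play.
# The cue uses the shared spooky prop attenuation and the Athena ambient props sound class, with a slightly lowered volume.
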SoundCue /Game/Athena/Sounds/Props/Prop_Bus_Flavor_Horn_02_Cue
root=Modulator_0
nodes [
 {id=Modulator_0, children [WavePlayer_0]}
 {id=WavePlayer_0, properties {soundWaveAssetPtr=&/Prop_Bus_Flavor_Horn_02}}
]
properties {
 attenuationSettings=/Game/Sounds/Attenuation/Fort_Prop_Spooky
 maxDistance=2000
 soundClassObject=&Fort_Ambient_Props_Athena
 volumeMultiplier=0.928
}
