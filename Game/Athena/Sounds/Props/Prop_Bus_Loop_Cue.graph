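# Sound cue for the looping ambient sound of a bus prop in Athena.
# A looping wave player plays Prop_Bus_Loop through a Modulator that slightly randomizes volume each time playback starts.
# The cue lowers overall volume, uses the Fort_Prop_Spooky attenuation asset and is routed to the Athena ambient props sound class.
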SoundCue /Game/Athena/Sounds/Props/Prop_Bus_Loop_Cue
root=Modulator_0
nodes [
 {id=Modulator_0, children [WavePlayer_0], properties {volumeMax=0.95, volumeMin=0.85}}
 {id=WavePlayer_0, properties {bLooping=true, soundWaveAssetPtr=&/Prop_Bus_Loop}}
]
properties {
 attenuationSettings=/Game/Sounds/Attenuation/Fort_Prop_Spooky
 maxDistance=2000
 soundClassObject=&Fort_Ambient_Props_Athena
 volumeMultiplier=0.65
}
