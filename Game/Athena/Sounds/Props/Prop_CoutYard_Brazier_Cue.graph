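# Sound cue for a courtyard brazier prop in Athena, playing a looping ring-of-fire sound on the ambient sound class.
# A quality-level node routes the first two levels to the standard loop and the last two to its OPT variant, both reused from quest audio.
# Overridden attenuation adds logarithmic falloff, occlusion with low-pass filtering, and listener focus that quiets the fire when off-focus.
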SoundCue /Game/Athena/Sounds/Props/Prop_CoutYard_Brazier_Cue
root=QualityLevel_0
nodes [
 {id=QualityLevel_0, children [WavePlayer_0,WavePlayer_0,WavePlayer_1,WavePlayer_1]}
 {id=WavePlayer_0, properties {bLooping=true, soundWaveAssetPtr=/Game/Athena/Sounds/Quests/Ring_Of_Fire_Loop}}
 {id=WavePlayer_1, properties {bLooping=true, soundWaveAssetPtr=/Game/Athena/Sounds/Quests/Ring_Of_Fire_Loop_OPT}}
]
properties {
 attenuationSettings=/Game/Sounds/Attenuation/Fort_Traps_Athena_Att
 bOverrideAttenuation=true
 maxDistance=1200
 soundClassObject=&Fort_Ambient_Athena
 volumeMultiplier=0.624
 attenuationOverrides {
  bEnableListenerFocus=true
  bEnableOcclusion=true
  distanceAlgorithm=Logarithmic
  falloffDistance=1200
  nonFocusVolumeAttenuation=0.3
  occlusionLowPassFilterFrequency=5000
  occlusionTraceChannel=ECC_GameTraceChannel11
  occlusionVolumeAttenuation=0.1
  stereoSpread=500
  attenuationShapeExtents {x=0}
 }
}
