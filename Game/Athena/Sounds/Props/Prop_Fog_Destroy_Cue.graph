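# Sound cue for the destruction sound of an Athena fog prop.
# A single wave player plays the Prop_Fog_Destroy wave directly, with no randomization or modulation.
# It uses spooky prop attenuation, the impacts sound class and a boosted volume.
# Concurrency allows one instance per owner, stopping the oldest when a new one starts.
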SoundCue /Game/Athena/Sounds/Props/Prop_Fog_Destroy_Cue
root=WavePlayer_0
nodes [{id=WavePlayer_0, properties {soundWaveAssetPtr=&/Prop_Fog_Destroy}}]
properties {
 attenuationSettings=/Game/Sounds/Attenuation/Fort_Prop_Spooky
 maxDistance=2000
 soundClassObject=&Fort_Impacts
 volumeMultiplier=2.8
 concurrencySet [/Game/Sounds/Concurrency_Rules/OneMaximum_Owner_StopOldest]
}
