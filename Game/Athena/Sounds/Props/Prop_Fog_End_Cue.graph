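# Sound cue for a fog prop in Athena, judging by names played when the fog ends or is destroyed.
# A mixer layers the end and destroy waves, weighted toward the destroy sound, and a modulator randomly pitches the blend down for variation.
# It routes to the Athena ambient props sound class with spooky prop attenuation, and concurrency allows one instance per owner, stopping the oldest.
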SoundCue /Game/Athena/Sounds/Props/Prop_Fog_End_Cue
root=Modulator_0
nodes [
 {id=Mixer_0, children [WavePlayer_0,WavePlayer_1], properties {inputVolume [0.3,0.7]}}
 {id=Modulator_0, children [Mixer_0], properties {pitchMax=0.9, pitchMin=0.7}}
 {id=WavePlayer_0, properties {soundWaveAssetPtr=&/Prop_Fog_End}}
 {id=WavePlayer_1, properties {soundWaveAssetPtr=&/Prop_Fog_Destroy}}
]
properties {
 attenuationSettings=/Game/Sounds/Attenuation/Fort_Prop_Spooky
 maxDistance=2000
 soundClassObject=&Fort_Ambient_Props_Athena
 volumeMultiplier=0.696
 concurrencySet [/Game/Sounds/Concurrency_Rules/OneMaximum_Owner_StopOldest]
}
