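# Sound cue for the Athena fog prop, playing a one-shot start sound that leads into a looping fog ambience.
# A mixer layers the quiet start wave with the louder loop, which enters after a short fixed delay, and feeds a default-setting modulator at the root.
# It uses spooky fog attenuation, the ambient props sound class, and owner-limited concurrency with distance-based volume scaling.
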SoundCue /Game/Athena/Sounds/Props/Prop_Fog_Loop_Cue
root=Modulator_0
nodes [
 {id=Delay_0, children [WavePlayer_0], properties {delayMax=0.1, delayMin=0.1}}
 {id=Mixer_0, children [WavePlayer_1,Delay_0], properties {inputVolume [0.1,1.7]}}
 {id=Modulator_0, children [Mixer_0]}
 {id=WavePlayer_0, properties {bLooping=true, soundWaveAssetPtr=&/Prop_Fog_Loop}}
 {id=WavePlayer_1, properties {soundWaveAssetPtr=&/Prop_Fog_Start}}
]
properties {
 attenuationSettings=/Game/Sounds/Attenuation/Fort_Prop_Fog_Spooky
 bOverrideConcurrency=true
 maxDistance=1400
 soundClassObject=&Fort_Ambient_Props_Athena
 volumeMultiplier=0.928
 concurrencyOverrides {bLimitToOwner=true, maxCount=3, volumeScale=0.7, volumeScaleMode=Distance}
}
