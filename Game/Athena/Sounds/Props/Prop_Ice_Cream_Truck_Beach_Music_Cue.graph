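# Sound cue that loops beach party music for the Athena ice cream truck prop, classed as Athena ambient audio.
# A single looping wave player feeds the output at reduced volume.
# Its own attenuation override sets the falloff, stereo spread and listener focus, which lowers volume for out-of-focus sources.
# It also declares a send to the music audio analysis submix, with the send level set to zero.
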SoundCue /Game/Athena/Sounds/Props/Prop_Ice_Cream_Truck_Beach_Music_Cue
root=WavePlayer_1
nodes [{id=WavePlayer_1, properties {bLooping=true, soundWaveAssetPtr=&/Beach_Party_Music_Loop_01}}]
properties {
 bOverrideAttenuation=true
 maxDistance=2048
 soundClassObject=&Fort_Ambient_Athena
 volumeMultiplier=0.376
 attenuationOverrides {
  bEnableListenerFocus=true
  falloffDistance=1536
  focusAzimuth=40
  nonFocusVolumeAttenuation=0.5
  stereoSpread=300
  attenuationShapeExtents=<512 1344 768>
 }
 soundSubmixSends [{minSendLevel=0, sendLevel=0, soundSubmix=&Fortnite_AudioAnalysis_Music}]
}
