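# Sound cue for a spider prop in Athena, playing as ambient prop audio with spooky-prop attenuation.
# A Random node picks one of three equally weighted spider wave players.
# A Modulator at the root randomizes volume and raises pitch by a random amount for variation between plays.
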SoundCue /Game/Athena/Sounds/Props/Prop_Spider_01_Cue
root=Modulator_0
nodes [
 {id=Modulator_0, children [Random_0], properties {pitchMax=1.4, pitchMin=1, volumeMax=2.1, volumeMin=1.8}}
 {id=Random_0, children [WavePlayer_0,WavePlayer_1,WavePlayer_2], properties {weights [1,1,1]}}
 {id=WavePlayer_0, properties {soundWaveAssetPtr=&/Prop_Spider_01}}
 {id=WavePlayer_1, properties {soundWaveAssetPtr=&/Prop_Spider_02}}
 {id=WavePlayer_2, properties {soundWaveAssetPtr=&/Prop_Spider_03}}
]
properties {
 attenuationSettings=/Game/Sounds/Attenuation/Fort_Prop_Spooky
 maxDistance=2000
 soundClassObject=&Fort_Ambient_Props_Athena
 volumeMultiplier=0.51
}
