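# Sound cue for a spider prop, the second numbered variant, assigned to the impacts sound class.
# A single wave player feeds a modulator that randomizes pitch from normal up to one and a half times on each play.
# The cue boosts volume and uses a shared spooky-prop attenuation for spatial falloff.
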SoundCue /Game/Athena/Sounds/Props/Prop_Spider_02_Cue
root=Modulator_0
nodes [
 {id=Modulator_0, children [WavePlayer_0], properties {pitchMax=1.5, pitchMin=1}}
 {id=WavePlayer_0, properties {soundWaveAssetPtr=&/Prop_Spider_02}}
]
properties {
 attenuationSettings=/Game/Sounds/Attenuation/Fort_Prop_Spooky
 maxDistance=2000
 soundClassObject=&Fort_Impacts
 volumeMultiplier=4
}
