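# Sound cue for the ambient bats loop used by props in Athena.
# A single wave player feeds the bats loop wave straight to the output, with no looping, random or modulation nodes.
# The cue uses the shared spooky prop attenuation and the Athena ambient props sound class, with its volume boosted above unity.
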SoundCue /Game/Athena/Sounds/Props/Props_Bats_Loop_Cue
root=WavePlayer_0
nodes [{id=WavePlayer_0, properties {soundWaveAssetPtr=&/Props_Bats_Loop}}]
properties {
 attenuationSettings=/Game/Sounds/Attenuation/Fort_Prop_Spooky
 maxDistance=2000
 soundClassObject=&Fort_Ambient_Props_Athena
 volumeMultiplier=1.856
}
